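# Child animation Blueprint for the QuiltLaze biped cosmetic companion, targeting the QuiltLaze_Biped_Skeleton.
# It adds no variables or animation logic of its own, and its event graph holds only an unconnected Try Get Pawn Owner node.
# The name suggests it inherits its behavior from a parent animation Blueprint, which is not identified here.
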AnimBlueprint /CosmeticCompanions/Assets/Biped/QuiltLaze/Blueprint/QuiltLaze_AnimBP_Child
target_skeleton=/CosmeticCompanions/Assets/Biped/QuiltLaze/Meshes/QuiltLaze_Biped_Skeleton
graphs [EventGraph {nodes [TryGetPawnOwner()]}]
variables []
