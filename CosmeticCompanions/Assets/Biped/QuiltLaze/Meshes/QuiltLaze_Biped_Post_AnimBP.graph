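# Post-process AnimBP for the QuiltLaze biped cosmetic companion, adding pose-asset blending and secondary bone motion to the incoming pose.
# A Pose Blend node applies QuiltLaze_Biped_PoseAsset, then chained root-relative AnimDynamics springs simulate four tail and two hair bones, the tail stiffer than the hair.
# A final weighted position constraint makes teeth_upper follow nine jaw, lip and cheek bones.
# All logic lives in the AnimGraph; the EventGraph is empty.
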AnimBlueprint /CosmeticCompanions/Assets/Biped/QuiltLaze/Meshes/QuiltLaze_Biped_Post_AnimBP
target_skeleton=&/QuiltLaze_Biped_Skeleton
graphs [
 AnimGraph {nodes [
   n1 = Input Pose()
   n2 = Get PoseAsset1() -> (PoseAsset1:Pose Asset Object Reference)
   n3 = Anim Dynamics - Bone: tail_01(
    ComponentPose=Local To Component(LocalPose=Pose!(PoseAsset:Pose Asset Object Reference=@n2, SourcePose:Pose Link Structure=@n1) -> (Pose:Pose Link Structure) {is=title AnimGraphNode_PoseBlendNode, blendOption=Linear})
    Alpha=1
    ExternalForce="0,0,0"
   ) {
    angularDampingOverride=1
    angularSpringConstant=13000
    bAngularSpring=true
    bOverrideAngularDamping=true
    bUseGravityOverride=true
    bUsePlanarLimit=false
    gravityScale=0.35
    simulationSpace=RootRelative
    alphaBoolBlend {blendOption=Linear}
    alphaScaleBiasClamp {bias=1, scale=-1}
    boundBone {boneName=tail_01}
    componentAppliedLinearAccClamp=<4000 4000 4000>
    componentLinearAccScale=<1 1 1>
    componentLinearVelScale=<1 1 1>
    gravityOverride {z=-980}
    physicsBodyDefinitions [{
      boundBone {boneName=tail_01}
      boxExtents=<27 27 27>
      constraintSetup {angularLimitsMax=<10 10 20>, angularLimitsMin=<-10 -30 -20>, angularTarget=<1 0 0>}
      localJointOffset=<4 0 0>
     }]
    simSpaceSettings {maxAngularAcceleration=40, maxAngularVelocity=8, simSpaceAngularAlpha=1}
   }
   n4 = Anim Dynamics - Bone: tail_02(ComponentPose=@n3, Alpha=1, ExternalForce="0,0,0") {
    angularDampingOverride=1
    angularSpringConstant=13000
    bAngularSpring=true
    bOverrideAngularDamping=true
    bUseGravityOverride=true
    bUsePlanarLimit=false
    gravityScale=0.35
    simulationSpace=RootRelative
    alphaBoolBlend {blendOption=Linear}
    alphaScaleBiasClamp {bias=1, scale=-1}
    boundBone {boneName=tail_02}
    componentAppliedLinearAccClamp=<4000 4000 4000>
    componentLinearAccScale=<1 1 1>
    componentLinearVelScale=<1 1 1>
    gravityOverride {z=-980}
    physicsBodyDefinitions [{
      boundBone {boneName=tail_02}
      boxExtents=<27 27 27>
      constraintSetup {angularLimitsMax=<10 10 20>, angularLimitsMin=<-10 -30 -20>, angularTarget=<1 0 0>}
      localJointOffset=<4 0 0>
     }]
    simSpaceSettings {maxAngularAcceleration=40, maxAngularVelocity=8, simSpaceAngularAlpha=1}
   }
   n5 = Anim Dynamics - Bone: tail_03(ComponentPose=@n4, Alpha=1, ExternalForce="0,0,0") {
    angularDampingOverride=1
    angularSpringConstant=13000
    bAngularSpring=true
    bOverrideAngularDamping=true
    bUseGravityOverride=true
    bUsePlanarLimit=false
    gravityScale=0.35
    simulationSpace=RootRelative
    alphaBoolBlend {blendOption=Linear}
    alphaScaleBiasClamp {bias=1, scale=-1}
    boundBone {boneName=tail_03}
    componentAppliedLinearAccClamp=<4000 4000 4000>
    componentLinearAccScale=<1 1 1>
    componentLinearVelScale=<1 1 1>
    gravityOverride {z=-980}
    physicsBodyDefinitions [{
      boundBone {boneName=tail_03}
      boxExtents=<27 27 27>
      constraintSetup {angularLimitsMax=<10 10 20>, angularLimitsMin=<-10 -20 -20>, angularTarget=<1 0 0>}
      localJointOffset=<4 0 0>
     }]
    simSpaceSettings {maxAngularAcceleration=40, maxAngularVelocity=8, simSpaceAngularAlpha=1}
   }
   n6 = Anim Dynamics - Bone: tail_04(ComponentPose=@n5, Alpha=1, ExternalForce="0,0,0") {
    angularDampingOverride=1
    angularSpringConstant=13000
    bAngularSpring=true
    bOverrideAngularDamping=true
    bUseGravityOverride=true
    bUsePlanarLimit=false
    gravityScale=0.35
    simulationSpace=RootRelative
    alphaBoolBlend {blendOption=Linear}
    alphaScaleBiasClamp {bias=1, scale=-1}
    boundBone {boneName=tail_04}
    componentAppliedLinearAccClamp=<4000 4000 4000>
    componentLinearAccScale=<1 1 1>
    componentLinearVelScale=<1 1 1>
    gravityOverride {z=-980}
    physicsBodyDefinitions [{
      boundBone {boneName=tail_04}
      boxExtents=<27 27 27>
      constraintSetup {angularLimitsMax=<10 10 20>, angularLimitsMin=<-10 -10 -20>, angularTarget=<1 0 0>}
      localJointOffset=<4 0 0>
     }]
    simSpaceSettings {maxAngularAcceleration=40, maxAngularVelocity=8, simSpaceAngularAlpha=1}
   }
   n7 = Anim Dynamics - Bone: dyn_hair_A_1(ComponentPose=@n6, Alpha=1, ExternalForce="0,0,0") {
    angularDampingOverride=1
    angularSpringConstant=7000
    bAngularSpring=true
    bOverrideAngularDamping=true
    bUseGravityOverride=true
    gravityScale=0.35
    simulationSpace=RootRelative
    alphaBoolBlend {blendOption=Linear}
    alphaScaleBiasClamp {bias=1, scale=-1}
    boundBone {boneName=dyn_hair_A_1}
    componentAppliedLinearAccClamp=<4000 4000 4000>
    componentLinearAccScale=<1 1 1>
    componentLinearVelScale=<1 1 1>
    gravityOverride {z=-980}
    physicsBodyDefinitions [{
      boundBone {boneName=dyn_hair_A_1}
      boxExtents=<16 16 16>
      constraintSetup {angularLimitsMax=<10 5 30>, angularLimitsMin=<-10 -15 -30>, angularTarget=<1 0 0>}
      localJointOffset=<4 0 0>
     }]
   }
   n8 = Anim Dynamics - Bone: dyn_hair_B(ComponentPose=@n7, Alpha=1, ExternalForce="0,0,0") {
    angularDampingOverride=1
    angularSpringConstant=7000
    bAngularSpring=true
    bOverrideAngularDamping=true
    bUseGravityOverride=true
    gravityScale=0.35
    simulationSpace=RootRelative
    alphaBoolBlend {blendOption=Linear}
    alphaScaleBiasClamp {bias=1, scale=-1}
    boundBone {boneName=dyn_hair_B}
    componentAppliedLinearAccClamp=<4000 4000 4000>
    componentLinearAccScale=<1 1 1>
    componentLinearVelScale=<1 1 1>
    gravityOverride {z=-980}
    physicsBodyDefinitions [{
      boundBone {boneName=dyn_hair_B}
      boxExtents=<16 16 16>
      constraintSetup {angularLimitsMax=<10 10 30>, angularLimitsMin=<-10 -10 -30>, angularTarget=<1 0 0>}
      localJointOffset=<4 0 0>
     }]
   }
   n9 = "Constraint - teeth_upper (PPPPPPPPP)"!(
    ComponentPose=@n8
    Alpha=1
    ConstraintWeights_0=0.0512
    ConstraintWeights_1:float=0.4355
    ConstraintWeights_2:float=0.1862
    ConstraintWeights_3:float=0.0565
    ConstraintWeights_4:float=0.0113
    ConstraintWeights_5:float=0.0002
    ConstraintWeights_6:float=0.1848
    ConstraintWeights_7:float=0.0623
    ConstraintWeights_8:float=0.0119
   ) -> (Pose) {
    boneToModify {boneName=teeth_upper}
    constraintSetup [
     {targetBone {boneName=reverse_jaw}}
     {targetBone {boneName=lip_upper_mid}}
     {targetBone {boneName=lip_upper_01_l}}
     {targetBone {boneName=lip_upper_02_l}}
     {targetBone {boneName=lip_upper_03_l}}
     {targetBone {boneName=cheek_tp_in_r}}
     {targetBone {boneName=lip_upper_01_r}}
     {targetBone {boneName=lip_upper_02_r}}
     {targetBone {boneName=lip_upper_03_r}}
    ]
    constraintWeights [1,1,1,1,1,1,1,1,1]
   }
   Output Pose(Result=Component To Local(ComponentPose=@n9))
  ]}
 EventGraph {links [], nodes []}
]
variables [PoseAsset1 {type=PoseAsset, default=&/QuiltLaze_Biped_PoseAsset}]
